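# One-shot stereo MetaSound for the SunRose River Styx activation sound, varying with the ChargeCount input fed from GCNL.
# A dash routing patch sends OnPlay by charge count to one-, two- or three-charge variant pools, with zero sharing the one-charge pool.
# Each pool picks a non-repeating random wave with random pitch shift.
# A gain-balanced mixer combines the three players, and any player finishing ends the source.
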MetaSoundSource /SunRoseEnvGameplay/Audio/SunRose_RiverStyx_Activate
inputs [
 Activate.OneCharge {type=WaveAsset:Array}
 Activate.ThreeCharges {type=WaveAsset:Array}
 Activate.TwoCharges {type=WaveAsset:Array}
 FromGCNL.ChargeCount {type=Int32}
 UE.Source.OnPlay {type=Trigger}
]
interfaces [UE.OutputFormat.Stereo:1.0, UE.Source.OneShot:1.0, UE.Source:1.1]
outputs [
 "UE.OutputFormat.Stereo.Audio:0" {type=Audio}
 "UE.OutputFormat.Stereo.Audio:1" {type=Audio}
 UE.Source.OneShot.OnFinished {type=Trigger}
]
pages [Default {
  input_defaults {
   FromGCNL.ChargeCount=2
   UE.Source.OnPlay=null
   Activate.OneCharge [
    &/SunRose_RiverStyx_Activate_One_01
    &/SunRose_RiverStyx_Activate_One_02
    &/SunRose_RiverStyx_Activate_One_03
    &/SunRose_RiverStyx_Activate_One_04
   ]
   Activate.ThreeCharges [
    &/SunRose_RiverStyx_Activate_Three_04
    &/SunRose_RiverStyx_Activate_Three_03
    &/SunRose_RiverStyx_Activate_Three_02
    &/SunRose_RiverStyx_Activate_Three_01
   ]
   Activate.TwoCharges [
    &/SunRose_RiverStyx_Activate_Two_01
    &/SunRose_RiverStyx_Activate_Two_02
    &/SunRose_RiverStyx_Activate_Two_03
    &/SunRose_RiverStyx_Activate_Two_04
   ]
  }
  nodes [
   n1 = /SunRoseEnvGameplay/Audio/Patches/SunRose_RiverStyx_Dash_TriggerRouteChargeCount(ChargeCount=^FromGCNL.ChargeCount, Input=^UE.Source.OnPlay)
   ArrayRandomGet_0 = Array.Random Get:WaveAsset:Array(In Array=^Activate.TwoCharges, Next=UE.Reroute(Value=@n1.Trigger2).Value) {also {defaults {No Repeats=3, Weights []}}}
   n2 = UE.RandomFloat(Next=@ArrayRandomGet_0.On Next) {also {defaults {Max=2, Min=-1}}}
   WavePlayer_0 = UE.Wave Player:Stereo(Wave Asset=@ArrayRandomGet_0.Value, Play=@n2.On Next, Pitch Shift=@n2.Value)
   TriggerAnyTriggerAny2_0 = "TriggerAny.Trigger Any (2)"(In 0=@n1.0, In 1=@n1.Trigger1)
   ArrayRandomGet_2 = Array.Random Get:WaveAsset:Array(In Array=^Activate.OneCharge, Next=@TriggerAnyTriggerAny2_0.Out) {also {defaults {No Repeats=3}}}
   n3 = UE.RandomFloat(Next=@ArrayRandomGet_2.On Next) {also {defaults {Max=2, Min=-1}}}
   WavePlayer_1 = UE.Wave Player:Stereo(Wave Asset=@ArrayRandomGet_2.Value, Play=@n3.On Next, Pitch Shift=@n3.Value)
   ArrayRandomGet_1 = Array.Random Get:WaveAsset:Array(In Array=^Activate.ThreeCharges, Next=UE.Reroute(Value=@n1.Trigger3).Value) {also {defaults {No Repeats=3, Weights []}}}
   n4 = UE.RandomFloat(Next=@ArrayRandomGet_1.On Next) {also {defaults {Max=2, Min=-1}}}
   WavePlayer_2 = UE.Wave Player:Stereo(Wave Asset=@ArrayRandomGet_1.Value, Play=@n4.On Next, Pitch Shift=@n4.Value)
   AudioMixerAudioMixerStereo3_0 = "AudioMixer.Audio Mixer (Stereo, 3)"(
    In 1 L=@WavePlayer_0.Out Left
    In 1 R=@WavePlayer_0.Out Right
    In 0 L=@WavePlayer_1.Out Left
    In 0 R=@WavePlayer_1.Out Right
    In 2 L=@WavePlayer_2.Out Left
    In 2 R=@WavePlayer_2.Out Right
   ) {also {defaults {Gain 1=0.8, Gain 2=0.7}}}
   TriggerAnyTriggerAny3_0 = "TriggerAny.Trigger Any (3)"(In 1=@WavePlayer_0.On Finished, In 0=@WavePlayer_1.On Finished, In 2=@WavePlayer_2.On Finished)
  ]
  links [
   {dst="@UE.OutputFormat.Stereo.Audio:0", src=AudioMixerAudioMixerStereo3_0.Out L}
   {dst="@UE.OutputFormat.Stereo.Audio:1", src=AudioMixerAudioMixerStereo3_0.Out R}
   {dst="@UE.Source.OneShot.OnFinished", src=TriggerAnyTriggerAny3_0.Out}
  ]
 }]
properties {
 assetImportData=&self:AssetImportData
 attenuationSettings=/SunRoseEnvGameplay/Audio/Settings/SunRose_RiverStyx_Activate
 maxDistance=1800
 numChannels=2
 outputFormat=Stereo
 sampleRate=48000
 soundClassObject=&Fort_Gadgets
 volume=0.6
}
